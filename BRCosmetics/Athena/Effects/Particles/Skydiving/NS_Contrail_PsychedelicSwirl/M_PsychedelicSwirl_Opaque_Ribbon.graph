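# Masked, lit Niagara ribbon and mesh-particle material for the psychedelic swirl skydiving contrail.
# Twist-sheared UVs sample Perlin noise offset by a particle time parameter, yielding normals and a blend of two tints and particle color.
# Sharpened Perlin times panned liner noise drives particle-colored emissive and a smooth dissolve, both attenuated by inverted Fresnel.
# Dynamic parameters control time offset, twist, exponent, emissive strength and dissolve.
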Material /BRCosmetics/Athena/Effects/Particles/Skydiving/NS_Contrail_PsychedelicSwirl/M_PsychedelicSwirl_Opaque_Ribbon
graph {
 properties {
  bAutomaticallySetUsageInEditor=false
  bUsedWithNiagaraMeshParticles=true
  bUsedWithNiagaraRibbons=true
  blendMode=BLEND_Masked
  usedShadingModels=Default Lit
 }
 nodes [
  Constant()
  n1 = TextureCoordinate()
  n2 = ComponentMask(None=@n1) {r=true}
  n3 = DynamicParameter() {defaultValue=<rgba 1 5 1 1>, paramNames [Time Offset, Twist, Exp, Emissive]}
  n4 = Add(A=AppendVector(A=Constant(), B=Multiply(A=@n2, B=@n3.Twist)), B=@n1)
  n5 = VectorParameter() {defaultValue=<rgba 0.1 2 0.2 1>}
  n6 = AppendVector(A=@n5.R, B=@n5.R)
  n7 = Constant2Vector() {g=1, r=0.1}
  n8 = TextureSample(UVs=Add(A=Multiply(A=@n4, B=@n6), B=Multiply(A=@n3.Time Offset, B=@n7))) {samplerType=SAMPLERTYPE_LinearColor, texture=/Game/Effects/Fort_Effects/Textures/Tile/T_PerlinRGB_Displacement}
  DeriveNormalZ(InXY=ConstantBiasScale(None=ComponentMask(None=@n8.R) {g=true, r=true}) {bias=-0.5, scale=2})
  n9 = ParticleColor()
  n10 = Multiply(A=@n8.B, B=@n8.B)
  n11 = Multiply(A=@n10, B=@n10)
  n12 = Panner(Coordinate=Multiply(A=@n4, B=AppendVector(A=@n5.B, B=@n5.B))) {speedX=0.1}
  n13 = Power(
   Base=Multiply(
    A=Multiply(A=@n11, B=@n11)
    B=TextureSample(UVs=@n12) {samplerType=SAMPLERTYPE_LinearColor, texture=/Game/Effects/Fort_Effects/Textures/Tile/LinerNoise64x64}.B
   )
   Exp=DynamicParameter() {defaultValue=<rgba 1 5 1 1>, paramNames [Time Offset, Twist, Exp, Emissive]}.Exp
  )
  n14 = OneMinus(None=Fresnel() {baseReflectFraction=0, exponent=0.5})
  n15 = Multiply(A=@n13, B=@n14)
  n16 = DynamicParameter() {defaultValue=<rgba 1 5 1 1>, paramNames [Time Offset, Twist, Exp, Emissive]}
  MaterialFunctionCall(Emissive=Multiply(A=@n9.RGB, B=Multiply(A=@n15, B=@n16.Emissive))) {materialFunction=&MF_FX_EmissiveAdjustments}
  n17 = VectorParameter() {parameterName=Param_1, defaultValue=<rgba 0.3677 0.0743 0.1357 1>}
  n18 = VectorParameter() {parameterName=Param_2, defaultValue=<rgba 0.4268 0.0199 0.7344 1>}
  n19 = Normalize(VectorInput=@n9.RGB)
  n20 = Multiply(A=@n9.A, B=@n8.B)
  Multiply(A=LinearInterpolate(A=LinearInterpolate(A=@n17.RGB, B=@n18.RGB, Alpha=@n8.G), B=Multiply(A=@n19, B=@n8.B), Alpha=Saturate(None=@n20)) {constB=0}) {constB=0.8}
  n21 = DynamicParameter() {parameterIndex=1, paramNames [Dissolve,Param2,Param3,Param4]}
  Multiply(
   A=Multiply(
    A=@n14
    B=MaterialFunctionCall!(Texture=@n13, Smooth, Steps=@n21.Dissolve) {materialFunction=/Game/Effects/Fort_Effects/Emotes/QuarrelFemale/MF_FX_Dissolve_Smooth_Sample}.Dissolve
   )
   B=ScalarParameter() {defaultValue=3}
  )
  Reroute(None=@n10)
 ]
}
