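# Masked, two-sided base material for Niagara mesh particles in the Psychedelic Swirl skydiving contrail.
# Dynamic parameters drive SplineThicken width, shear and scroll of Perlin noise, and emissive strength.
# Noise times a V gradient feeds particle-colored emissive and, with particle alpha and a sphere mask, a mask term.
# A sine sway rotates vertices about the particle pivot on a slowly turning, per-particle randomized axis, time-offset along UV1 height.
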Material /BRCosmetics/Athena/Effects/Particles/Skydiving/NS_Contrail_PsychedelicSwirl/M_PsychedelicSwirl_SplineThicken_Base
graph {
 properties {
  bAutomaticallySetUsageInEditor=false
  bUsedWithNiagaraMeshParticles=true
  blendMode=BLEND_Masked
  twoSided=true
  usedShadingModels=Default Lit
 }
 nodes [
  n1 = DynamicParameter() {defaultValue=<rgba 1 1.5 1 1>, paramNames [SplineWidth,Progression,Emissive,Time]}
  MaterialFunctionCall(WidthBase=@n1.SplineWidth, WidthTip=@n1.SplineWidth) {materialFunction=&SplineThicken}
  n2 = TextureCoordinate()
  n3 = ComponentMask(None=@n2) {r=true}
  n4 = DynamicParameter() {defaultValue=<rgba 1 1.5 1 1>, paramNames [SplineWidth,Progression,Emissive,Time]}
  n5 = Add(A=AppendVector(A=Constant(), B=Multiply(A=@n3, B=@n4.Progression)), B=@n2)
  n6 = VectorParameter() {defaultValue=<rgba 0.1 1 0.1 0.5>}
  n7 = AppendVector(A=@n6.R, B=@n6.R)
  n8 = Constant2Vector() {g=1, r=0.1}
  n9 = TextureSample(UVs=Add(A=Multiply(A=@n5, B=@n7), B=Multiply(A=@n4.Time, B=@n8))) {samplerType=SAMPLERTYPE_LinearColor, texture=/Game/Effects/Fort_Effects/Textures/Tile/T_PerlinRGB_Displacement}
  n10 = Multiply(A=@n9.B, B=ComponentMask(None=TextureCoordinate()) {g=true})
  n11 = ParticleColor()
  MaterialFunctionCall(Emissive=Multiply(
    A=Multiply(A=@n11.RGB, B=@n10)
    B=DynamicParameter() {defaultValue=<rgba 1 1.5 1 1>, paramNames [SplineWidth,Progression,Emissive,Time]}.Emissive
   )) {materialFunction=&MF_FX_EmissiveAdjustments}
  n12 = TextureCoordinate()
  n13 = Constant() {r=0.5}
  Multiply(A=Multiply(A=Multiply(A=@n11.A, B=@n10), B=SphereMask(A=@n12, B=@n13, Radius=@n13) {hardnessPercent=0})) {constB=5}
  Multiply(A=@n11.RGB, B=@n9.B)
  n14 = Constant3Vector() {desc=rotation axis' rotation axis, constant=<rgba 1 0 0 0>}
  n15 = DynamicParameter() {defaultValue=<rgba 1 1.5 1 1>, paramNames [SplineWidth,Progression,Emissive,Time]}
  n16 = Add(A=Time(), B=ParticleRandom())
  n17 = ScalarParameter() {defaultValue=0.2, parameterName=Axis Rotation Speed}
  n18 = Normalize(VectorInput=Transform(None=Constant3Vector() {desc=rotation axis, constant=<rgba 0 1 0 0>}) {transformSourceType=TRANSFORMSOURCE_Local})
  n19 = RotateAboutAxis(
   NormalizedRotationAxis=Normalize(VectorInput=Transform(None=@n14) {transformSourceType=TRANSFORMSOURCE_Local})
   RotationAngle=Add(A=@n15.Time, B=Multiply(A=@n16, B=@n17))
   PivotPoint=Constant()
   Position=@n18
  )
  n20 = Time()
  n21 = ScalarParameter() {defaultValue=3, parameterName=Rotation Speed}
  n22 = ComponentMask(None=TextureCoordinate() {coordinateIndex=1}) {g=true}
  n23 = Subtract(
   A=Add(A=Multiply(A=@n20, B=@n21), B=@n15.Time)
   B=Multiply(A=OneMinus(None=@n22), B=ScalarParameter() {defaultValue=0.3, parameterName=Height Time Offset})
  )
  RotateAboutAxis(
   NormalizedRotationAxis=Normalize(VectorInput=Add(A=@n19, B=@n18))
   RotationAngle=Multiply(A=Sine(None=@n23), B=ScalarParameter() {defaultValue=0.1, parameterName=Rotation Angle Max})
   PivotPoint=MaterialFunctionCall() {materialFunction=&ObjectPivotPoint}.Mesh Particle Pivot Location
   Position=WorldPosition().XYZ
  )
 ]
}
