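# Translucent unlit wisp material for Niagara mesh particles in the Psychedelic Swirl skydiving contrail.
# SplineThicken sets mesh width from a dynamic SplineWidth parameter, while Progression and Scale shift and stretch clamped, V-flipped UVs into a streak mask.
# Particle-random offset Perlin noise breaks up that mask, and the mask's fourth power adds a bright core.
# Particle color tints the sum through an emissive adjustment function, and its alpha scales the noisy mask.
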Material /BRCosmetics/Athena/Effects/Particles/Skydiving/NS_Contrail_PsychedelicSwirl/M_PsychedelicSwirl_SplineThicken_Wisp
graph {
 properties {
  bAutomaticallySetUsageInEditor=false
  bUsedWithNiagaraMeshParticles=true
  blendMode=BLEND_Translucent
  shadingModel=MSM_Unlit
  twoSided=true
  usedShadingModels=Unlit
 }
 nodes [
  n1 = DynamicParameter() {defaultValue=<rgba 1 1.5 1 1>, paramNames [SplineWidth,Progression,Scale,Param4]}
  MaterialFunctionCall(WidthBase=@n1.SplineWidth, WidthTip=@n1.SplineWidth) {materialFunction=&SplineThicken}
  n2 = Add(A=ParticleRandom(), B=TextureCoordinate())
  n3 = Constant2Vector() {g=0.05, r=0.3}
  n4 = Constant()
  n5 = DynamicParameter() {defaultValue=<rgba 1 1.5 1 1>, paramNames [SplineWidth,Progression,Scale,Param4]}
  n6 = Multiply(A=Add(A=TextureCoordinate(), B=AppendVector(A=@n4, B=@n5.Progression)), B=AppendVector(A=Constant() {r=1}, B=@n5.Scale))
  n7 = TextureSample(UVs=Multiply(A=Saturate(None=@n6), B=Constant2Vector() {g=-1, r=1})) {texture=/BRCosmetics/Athena/Effects/Particles/Skydiving/NS_Contrail_EnergyNet/T_Contrail_EnergyNet_Mask01}
  n8 = Multiply(
   A=TextureSample(UVs=Multiply(A=@n2, B=@n3)) {samplerType=SAMPLERTYPE_LinearColor, texture=/Game/Effects/Fort_Effects/Textures/Tile/T_PerlinRGB_Displacement}.B
   B=@n7.R
  )
  n9 = Multiply(A=@n7.R, B=@n7.R)
  n10 = Multiply(A=@n9, B=@n9)
  n11 = ParticleColor()
  MaterialFunctionCall(Emissive=Multiply(A=Add(A=@n8, B=@n10), B=@n11.RGB)) {materialFunction=&MF_FX_EmissiveAdjustments}
  Multiply(A=@n8, B=@n11.A)
 ]
}
