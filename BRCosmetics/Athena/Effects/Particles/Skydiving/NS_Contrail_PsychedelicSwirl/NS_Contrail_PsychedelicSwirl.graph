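# Skydiving contrail cosmetic that trails swirling, color-cycling ribbons from the character's skeletal mesh sockets.
# NE_LocalSpaceRibbon builds spring-driven pendulum chains that lean against owner velocity, offset by a time-animated location module.
# Other emitters read those chain particles through attribute readers to color flickering glow lights and to orient or move spinning meshes and sprites along the ribbons.
# A smoothed dive alpha scales width, brightness, spin and spawn rate, with front-end overrides adjusting the look.
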NiagaraSystem /BRCosmetics/Athena/Effects/Particles/Skydiving/NS_Contrail_PsychedelicSwirl/NS_Contrail_PsychedelicSwirl
dependencies {
 usedDataInterfaces [
  &NiagaraDataInterfaceCurve
  &NiagaraDataInterfaceColorCurve
  &NiagaraDataInterfaceCurlNoise
  &NiagaraDataInterfaceParticleRead
  &NiagaraDataInterfaceVectorCurve
  &NiagaraDataInterfaceSpriteRendererInfo
  &NiagaraDataInterfaceSkeletalMesh
 ]
 usedDynamicInputs [&Multiply_Vector2D_ByFloat, /Niagara/DynamicInputs/Lerp/SmoothLerpOverTime_Float]
 usedModules [
  &self:SystemSpawnScript.SystemScriptSource.SystemScriptGraph.NiagaraNodeAssignment_1.SetVariables_#1
  &self:SystemSpawnScript.SystemScriptSource.SystemScriptGraph.NiagaraNodeAssignment_0.SetVariables_#2
  &self:SystemSpawnScript.SystemScriptSource.SystemScriptGraph.NiagaraNodeAssignment_4.SetVariables_#3
  &self:SystemSpawnScript.SystemScriptSource.SystemScriptGraph.NiagaraNodeAssignment_3.SetVariables_#4
  &self:SystemSpawnScript.SystemScriptSource.SystemScriptGraph.NiagaraNodeAssignment_2.SetVariables_#5
  &SystemState
  &InterpolateOverTime
 ]
}
emitters {
 Default_LightAndGlowFlicker_Loop {
  emitterUpdateScript [
   &EmitterState(
    Inactive Response=::NewEnumerator1
    Life Cycle Mode=::NewEnumerator1
    Loop Behavior=::NewEnumerator1
    Loop Duration Mode=::NewEnumerator1
    UseLoopDelay=false
   )
   &SpawnBurst_Instantaneous(Spawn Count=4, Use Spawn Probability=false, UseLoopCountLimit=false)
   /Niagara/Noise/Waveform(
    EnableExponent1=false
    GlobalClampMax=false
    GlobalClampMin=false
    UseInput1=false
    UsePhase1=false
    "Waveform [1]":ENiagara_Waveforms=::NewEnumerator12
    "[1] Frequency"=7
   )
   /SetVariables(Emitter.Flicker:NiagaraFloat=&Add_Float(A=&Multiply_Float(A=$Emitter.Waveform.Waveform, B=0.5), B=0.5)) {isSetParameters=true}
  ]
  particleSpawnScript [
   &InitializeParticle(Sprite Size Mode=::NewEnumerator3, Uniform Sprite Size=50, UsePositionOffset=false, Write Material Random=true)
   &DynamicMaterialParameters(
    Index 0 Param 1=0.2
    Index 0 Param 2=1
    Module.Param0WriteEnabled=true
    Module.Param1WriteEnabled=true
    Module.Param2WriteEnabled=true
    Module.Param3WriteEnabled=true
   )
   &PartitionParticles(Number Of Partitions=4, Particles Per Partition=1, Partition Mode:ENiagara_PartitionMode=::NewEnumerator1)
  ]
  particleUpdateScript [
   &ParticleState(Kill Particles When Lifetime Has Elapsed=false, Let Infinitely Lived Particles Die When Emitter Deactivates=true)
   &self:SampleLocalRibbon(
    Particle Attribute Reader=DataInterface() {propertyValues {EmitterBinding {bindingMode=Other, emitterName=NE_LocalSpaceRibbon}}}
    ParticlePerRibbon:NiagaraInt32=$System.ParticlePerRibbon
    Partition:NiagaraInt32=&MakeIntFromFloat(Float=&Abs_Float(Float:NiagaraFloat=&Subtract_Float(A=&Multiply_FloatByInt(Float=0.75, Integer=$Particles.PartitionParticles.Partition), B=1.25)))
   ) {module=SampleLocalRibbon}
   &SkeletalMeshLocation(
    Bone / Socket Index=$Particles.PartitionParticles.Partition
    Bone Sampling Mode=::NewEnumerator4
    MaskSpawnGroup=false
    Sampled Position Offset=&Lerp_Vector(A=<0 0 0>, B=<0 0 0>, Alpha=&MakeFloatFromInt(Int=$Particles.PartitionParticles.Partition))
    Skeletal Mesh=$System.Skeletal Mesh
   ) {module=SkeletalMeshLocation001}
   &ScaleColor(
    Scale Alpha=&Multiply_Float(A=$Emitter.Flicker, B=0.3)
    Scale RGB=&MakeVectorFromLinearColorRGB(Color=&MultiplyLinearColor(
      Linear Color=$Output.SampleLocalRibbon.RibbonColor
      Scale Factor=&MakeVector4_FromLinearColor(LinearColor=&LerpLinearColors(StartColor=<rgba 0.3 0.3 0.3 1>, EndColor=<rgba 0.5 0.5 0.5 1>, LerpFactor=$System.SmoothDiveAlpha))
     ))
    ScaleA=true
    ScaleRGB=true
   )
   &FN_Light_Attributes(
    Color=$Particles.Color
    Create Color Attribute=true
    Create Exponent Attribute=true
    Create Radius Attribute=true
    Exponent=5
    Radius=100
    Scale Color=&Multiply_Float(A=7, B=$Emitter.Flicker)
   )
   &DynamicMaterialParameters(
    Index 0 Param 1=1
    Index 0 Param 2=1
    Index 0 Param 3=1
    Index 0 Param 4=&Lerp_Float(A=1, B=1, Alpha=$System.SmoothDiveAlpha)
    Module.Index0Param0Write=true
    Module.Index0Param1Write=true
    Module.Index0Param2Write=true
    Module.Index0Param3Write=true
   ) {module=DynamicMaterialParameters001}
   &CameraOffset(Camera Offset Amount=50)
  ]
  rendererClasses [&NiagaraSpriteRendererProperties, &NiagaraLightRendererProperties]
  renderers [
   {rendererClass=&NiagaraSpriteRendererProperties}
   {rendererClass=&NiagaraLightRendererProperties, rendererIndex=1}
   {rendererClass=&NiagaraSpriteRendererProperties, rendererIndex=2}
  ]
 }
 NE_Empty {
  emitterUpdateScript [
   &EmitterState(Life Cycle Mode=::NewEnumerator1, Loop Behavior=::NewEnumerator1, Loop Duration Mode=::NewEnumerator1, UseLoopDelay=false) {module=EmitterLifeCycle}
   &SpawnBurst_Instantaneous(Spawn Count=6, Use Spawn Probability=false, UseLoopCountLimit=false)
  ]
  particleSpawnScript [
   &InitializeParticle(Color Mode=::NewEnumerator0, Lifetime=5, UsePositionOffset=false, Write Material Random=true)
   &PartitionParticles(Normalized Index In Partition Includes One=true, Particles Per Partition=3, Partition Mode:ENiagara_PartitionMode=::NewEnumerator1)
   &InitialMeshOrientation(
    Mesh Orientation Mode=::NewEnumerator2
    Orientation Coordinate Space=::Local
    Orientation Vector=<0 0 1>
    Rotation=&Lerp_Vector(A=<0 0 0>, B=<0 0 1>, Alpha=$Particles.PartitionParticles.Normalized Index In Partition)
    Use Rotation Vector=true
   )
   /SetVariables(Particles.Scale=&Multiply_Vector(
     A=&Lerp_Vector(A=<0.1 0.1 0.4>, B=<0.1 0.1 0.2>, Alpha=$Particles.PartitionParticles.Normalized Index In Partition)
     B=&Lerp_Vector(A=<1 -1 1>, B=<1 1 1>, Alpha=&MakeFloatFromInt(Int=$Particles.PartitionParticles.Partition))
    )) {isSetParameters=true, module=SetVariables_#6}
   &Color(Color=&ColorFromCurve(
     ColorCurve=DataInterface() {propertyValues {
       AlphaCurve {keys [<at 0 1 RCIM_Cubic>,<at 1 2 RCIM_Cubic>]}
       BlueCurve {keys [
         <at 0.175 0.0386 RCIM_Cubic>
         {arriveTangent=1.2844, interpMode=RCIM_Cubic, leaveTangent=1.2844, time=0.2095, value=0.4253}
         <at 0.4761 0.4253>
         <at 0.5418 1>
        ]}
       GreenCurve {keys [
         <at 0.175 0.3219 RCIM_Cubic>
         {arriveTangent=-0.7751, interpMode=RCIM_Cubic, leaveTangent=-0.7751, time=0.2095, value=0.0885}
         <at 0.4761 0.0885>
         <at 0.5418 0.0444>
        ]}
       RedCurve {keys [<at 0.175 1 RCIM_Cubic>,<at 0.2095 1 RCIM_Cubic>,<at 0.4761 1>,<at 0.5418 0.4687>]}
      }}
     CurveIndex=$Particles.PartitionParticles.Normalized Index In Partition
    ))
  ]
  particleUpdateScript [
   &ParticleState(Kill Particles When Lifetime Has Elapsed=false)
   &ScaleMeshSize(Scale Factor=&Lerp_Vector(A=<1 1 0.7>, B=<1.5 1.5 1.5>, Alpha=$System.SmoothDiveAlpha))
   &self:SampleLocalRibbon(
    Particle Attribute Reader=DataInterface() {propertyValues {EmitterBinding {bindingMode=Other, emitterName=NE_LocalSpaceRibbon}}}
    ParticlePerRibbon:NiagaraInt32=$System.ParticlePerRibbon
    Partition:NiagaraInt32=$Particles.PartitionParticles.Partition
   ) {module=SampleLocalRibbon}
   &SkeletalMeshLocation(
    Bone / Socket Index=&Add_Integer(A=$Particles.PartitionParticles.Partition, B=2)
    Bone Sampling Mode=::NewEnumerator4
    MaskSpawnGroup=false
    Orientation Sampling=::NewEnumerator1
    Sampled Position Offset=&Lerp_Vector(A=<5 0 2>, B=<-5 0 -2>, Alpha=&MakeFloatFromInt(Int=$Particles.PartitionParticles.Partition))
    Skeletal Mesh=$System.Skeletal Mesh
   ) {module=SkeletalMeshLocation001}
   &UpdateMeshOrientation(
    Rotation Rate=&Multiply_Float(
     A=&Lerp_Float(A=1, B=-1, Alpha=&MakeFloatFromInt(Int=$Particles.PartitionParticles.Partition))
     B=&Lerp_Float(A=1.2, B=2, Alpha=$System.SmoothDiveAlpha)
    )
    Rotation Vector=<0 0 1>
   ) {module=UpdateMeshOrientation001}
   &UpdateMeshOrientation(
    Facing Direction=$Particles.RibbonDirectionVector
    Facing Interpolation=false
    Orientation Method=::NewEnumerator1
    Override Quat=false
    TransformFacing=false
    TransformUp=false
    Up Direction:Vector3f=<0 1 0>
   )
   &SolveForcesAndVelocity(Clamp Velocity=false, Limit Acceleration=false)
   &DynamicMaterialParameters(
    Index 0 Param 1=20
    Index 0 Param 2=1
    Index 0 Param 3=&Multiply_Float(A=$Engine.System.Age, B=1)
    Module.Index0Param0Write=true
    Module.Index0Param1Write=true
    Module.Index0Param2Write=true
    Module.Index0Param3Write=true
   )
   &Color(
    Color=&LerpLinearColors(
     StartColor=$Particles.Initial.Color
     EndColor=$Output.SampleLocalRibbon.RibbonColor
     LerpFactor=&Lerp_Float(A=0.5, B=0.4, Alpha=$System.SmoothDiveAlpha)
    )
    Scale Alpha=&Lerp_Float(A=0.6, B=0.5, Alpha=$System.SmoothDiveAlpha)
   ) {module=Color001}
  ]
  rendererClasses [&NiagaraMeshRendererProperties]
  renderers [{rendererClass=&NiagaraMeshRendererProperties, rendererIndex=0}]
 }
 NE_Empty001 {
  emitterUpdateScript [
   &EmitterState(Life Cycle Mode=::NewEnumerator1, Loop Behavior=::NewEnumerator1, Loop Duration Mode=::NewEnumerator1, UseLoopDelay=false) {module=EmitterLifeCycle}
   &SpawnBurst_Instantaneous(Spawn Count=6, Use Spawn Probability=false, UseLoopCountLimit=false)
  ]
  particleSpawnScript [
   &InitializeParticle(Color Mode=::NewEnumerator0, Lifetime=5, UsePositionOffset=false, Write Material Random=true)
   &PartitionParticles(Normalized Index In Partition Includes One=true, Particles Per Partition=3, Partition Mode:ENiagara_PartitionMode=::NewEnumerator1)
   &InitialMeshOrientation(
    Mesh Orientation Mode=::NewEnumerator2
    Orientation Coordinate Space=::Local
    Orientation Vector=<0 0 1>
    Rotation=&Lerp_Vector(
     A=<0 0 0>
     B=&Lerp_Vector(A=<0 0 1>, B=<0 0 -1>, Alpha=&MakeFloatFromInt(Int=$Particles.PartitionParticles.Partition))
     Alpha=$Particles.PartitionParticles.Normalized Index In Partition
    )
    Use Rotation Vector=true
   )
   /SetVariables(Particles.Scale=&Multiply_Vector(
     A=&Lerp_Vector(A=<0.175 0.175 0.45>, B=<0.1 0.1 0.1>, Alpha=$Particles.PartitionParticles.Normalized Index In Partition)
     B=&Lerp_Vector(A=<1 1 1>, B=<1 -1 1>, Alpha=&MakeFloatFromInt(Int=$Particles.PartitionParticles.Partition))
    )) {isSetParameters=true, module=SetVariables_#6}
   &Color(Color=&ColorFromCurve(
     ColorCurve=DataInterface() {propertyValues {
       AlphaCurve {keys [<at 0 1 RCIM_Cubic>,<at 1 2 RCIM_Cubic>]}
       BlueCurve {keys [
         <at 0.175 0.2314 RCIM_Cubic>
         {arriveTangent=0.644, interpMode=RCIM_Cubic, leaveTangent=0.644, time=0.2095, value=0.4253}
         <at 0.4761 0.4253>
         <at 0.5418 1>
        ]}
       GreenCurve {keys [
         <at 0.175 0.7898 RCIM_Cubic>
         {arriveTangent=-2.3293, interpMode=RCIM_Cubic, leaveTangent=-2.3293, time=0.2095, value=0.0885}
         <at 0.4761 0.0885>
         <at 0.5418 0.0444>
        ]}
       RedCurve {keys [<at 0.175 1 RCIM_Cubic>,<at 0.2095 1 RCIM_Cubic>,<at 0.4761 1>,<at 0.5418 0.4687>]}
      }}
     CurveIndex=$Particles.PartitionParticles.Normalized Index In Partition
    ))
  ]
  particleUpdateScript [
   &ParticleState(Kill Particles When Lifetime Has Elapsed=false)
   &ScaleMeshSize(Scale Factor=&Lerp_Vector(A=<1 1 1>, B=<1.2 1.2 2>, Alpha=$System.SmoothDiveAlpha))
   &self:SampleLocalRibbon(
    Particle Attribute Reader=DataInterface() {propertyValues {EmitterBinding {bindingMode=Other, emitterName=NE_LocalSpaceRibbon}}}
    ParticlePerRibbon:NiagaraInt32=$System.ParticlePerRibbon
    Partition:NiagaraInt32=$Particles.PartitionParticles.Partition
   ) {module=SampleLocalRibbon}
   &UpdateMeshOrientation(
    Rotation Rate=&Multiply_Float(
     A=&Lerp_Float(A=-1, B=1, Alpha=&MakeFloatFromInt(Int=$Particles.PartitionParticles.Partition))
     B=&Lerp_Float(A=1.6, B=2.5, Alpha=$System.SmoothDiveAlpha)
    )
    Rotation Vector=<0 0 1>
   ) {module=UpdateMeshOrientation001}
   &UpdateMeshOrientation(
    Facing Direction=$Particles.RibbonDirectionVector
    Facing Interpolation=false
    Orientation Method=::NewEnumerator1
    Override Quat=false
    TransformFacing=false
    TransformUp=false
    Up Direction:Vector3f=<0 1 0>
   )
   &SolveForcesAndVelocity(Clamp Velocity=false, Limit Acceleration=false)
   &DynamicMaterialParameters(
    Index 0 Param 1=25
    Index 0 Param 2=2
    Index 0 Param 3=&Multiply_Float(A=$Engine.System.Age, B=1)
    Module.Index0Param0Write=true
    Module.Index0Param1Write=true
    Module.Index0Param2Write=true
    Module.Index0Param3Write=true
   )
   &Color(
    Color=&LerpLinearColors(
     StartColor=$Particles.Initial.Color
     EndColor=$Output.SampleLocalRibbon.RibbonColor
     LerpFactor=&Lerp_Float(A=0.1, B=1, Alpha=$System.SmoothDiveAlpha)
    )
    Scale Alpha=&Lerp_Float(A=0.6, B=0.8, Alpha=$System.SmoothDiveAlpha)
   ) {module=Color001}
  ]
  rendererClasses [&NiagaraMeshRendererProperties]
  renderers [{rendererClass=&NiagaraMeshRendererProperties, rendererIndex=0}]
 }
 NE_LocalSpaceRibbon {
  emitterUpdateScript [
   &EmitterState(
    Inactive Response=::NewEnumerator1
    Life Cycle Mode=::NewEnumerator1
    Loop Behavior=::NewEnumerator1
    Loop Duration Mode=::NewEnumerator1
    UseLoopDelay=false
   ) {module=EmitterLifeCycle}
   &SpawnBurst_Instantaneous(
    Spawn Count=&Multiply_Int(A=$System.ChainNumber, B=$System.ParticlePerRibbon)
    Use Spawn Probability=false
    UseLoopCountLimit=false
   )
  ]
  particleSpawnScript [
   &InitializeParticle(Lifetime=5, Sprite Size Max=<20 20>, Sprite Size Mode=::NewEnumerator2, UsePositionOffset=false, Write Material Random=true)
   &PartitionParticles(
    Number Of Partitions=$System.ChainNumber
    Particles Per Partition:NiagaraInt32=$System.ParticlePerRibbon
    Partition Mode:ENiagara_PartitionMode=::NewEnumerator1
   )
   /SetVariables(Particles.NormalizedChainPosition:NiagaraFloat=$Particles.PartitionParticles.Normalized Index In Partition) {isSetParameters=true}
   &SkeletalMeshLocation(
    Bone / Socket Index=$Particles.PartitionParticles.Partition
    Bone Sampling Mode=::NewEnumerator4
    MaskSpawnGroup=false
    Sampled Position Offset=&Lerp_Vector(A=<0 0 0>, B=<0 0 150>, Alpha=$Particles.NormalizedChainPosition)
    Sampled Position Offset Coordinate Space:ENiagaraSkelMeshTransforms=::NewEnumerator1
    Skeletal Mesh=$System.Skeletal Mesh
   ) {enabled=false}
  ]
  particleUpdateScript [
   &ParticleState(Kill Particles When Lifetime Has Elapsed=false)
   &SkeletalMeshLocation(
    Bone / Socket Index=$Particles.PartitionParticles.Partition
    Bone Sampling Mode=::NewEnumerator4
    MaskSpawnGroup=false
    Position Sampling=::NewEnumerator2
    Sampled Position Offset=&Lerp_Vector(A=<2 5 0>, B=<-2 -5 0>, Alpha=&MakeFloatFromInt(Int=$Particles.PartitionParticles.Partition))
    Skeletal Mesh=$System.Skeletal Mesh
   ) {module=SkeletalMeshLocation001}
   &PendulumSetup(
    Module.Enable Spring Driver=true
    Pendulum Length:NiagaraFloat=&Lerp_Float(
     A=0
     B=$System.ChainLength
     Alpha=&FloatFromCurve(
      FloatCurve=DataInterface() {propertyValues {Curve {keys [<at 0 0>, {arriveTangent=-0.0703, interpMode=RCIM_Cubic, leaveTangent=-0.0703, tangentMode=RCTM_User, time=1, value=1}]}}}
      CurveIndex=$Particles.NormalizedChainPosition
      Scale Curve=&FloatFromCurve(
       FloatCurve=DataInterface() {propertyValues {Curve {keys [
           {arriveTangent=0.6092, interpMode=RCIM_Cubic, leaveTangent=0.6092, tangentMode=RCTM_User, time=0, value=0}
           {tangentMode=RCTM_User, time=1, value=1}
          ]}}}
       CurveIndex=$Emitter.Age
      )
     )
    )
    Pendulum Pivot:Vector3f=&AddVectorToPosition(
     Position=&AddVectorToPosition(Position=$Particles.SkeletalMeshLocation001.SampledPosition, Vector=<0 0 0>)
     Vector=&TransformVector(Vector=<0 0 0>, Source Space=::World, Destination Space=::Simulation)
    )
    Pendulum Pivot Offset Coordinate Space:ENiagaraCoordinateSpace=::Local
    Pendulum Rest Axis:Vector3f=&TransformVector(
     Vector=&Add_Vector(
      A=&Lerp_Vector(
       A=&Add_Vector(
        A=&Multiply_VectorByFloat(
         Vector=&NormalizeVector(Vector To Normalize=$Engine.Owner.Velocity, Safe Normalize=true, Safe Fallback Threshold=0, Safe Fallback Vector=<0 0 0.001>)
         Float=-0.5
        )
        B=&TransformVector(
         Vector=&RotateVectorByQuaternion(
          Vector To Rotate=&Multiply_VectorByFloat(
           Vector=<0.5 1 0>
           Float=&Lerp_Float(A=1, B=-1, Alpha=&MakeFloatFromInt(Int=$Particles.PartitionParticles.Partition))
          )
          Quaternion=$Particles.SkeletalMeshLocation001.SampledRotation
         )
         Destination Space=::World
        )
       )
       B=&Lerp_Vector(
        A=&Multiply_VectorByFloat(
         Vector=&NormalizeVector(Vector To Normalize=$Engine.Owner.Velocity, Safe Normalize=true, Safe Fallback Threshold=0, Safe Fallback Vector=<0 0 0.001>)
         Float=-1
        )
        B=<0 0 0.1>
        Alpha=$System.SmoothDiveAlpha
       )
       Alpha=&Lerp_Float(
        A=&FloatFromCurve(
         FloatCurve=DataInterface() {propertyValues {Curve {keys [
             {arriveTangent=1.8207, interpMode=RCIM_Cubic, leaveTangent=1.8207, tangentMode=RCTM_User, time=0, value=0}
             {arriveTangent=-0.078, interpMode=RCIM_Cubic, leaveTangent=-0.078, tangentMode=RCTM_User, time=1, value=1}
            ]}}}
         CurveIndex=$Particles.NormalizedChainPosition
        )
        B=1
        Alpha=&FN_FrontEndFloatModifier(Gameplay=0, Front End=1)
       )
      )
      B=&TransformVector(
       Vector=&FN_FrontEndVector3Modifier(Gameplay=<0 0 0>, Front End=&Lerp_Vector(A=<0 -0.1 0.2>, B=<0 -0.2 0.2>, Alpha=$System.SmoothDiveAlpha))
       Destination Space=::World
      )
     )
     Source Space=::World
     Destination Space=::Simulation
    )
    "Spring Driven Constraint (Tightness)"=30
   )
   &LinearForce(
    Coordinate Space:ENiagaraCoordinateSpace=::World
    Force=&Multiply_VectorByFloat(Vector=$Engine.Owner.Velocity, Float=-5)
   )
   &SolveForcesAndVelocity(Clamp Velocity=false, Limit Acceleration=false)
   &PendulumConstraint()
   &CalculateAccurateVelocity()
   /SetVariables(Particles.RibbonWidth=&FloatFromCurve(
     FloatCurve=DataInterface() {propertyValues {Curve {keys [
         {arriveTangent=8.0134, interpMode=RCIM_Cubic, leaveTangent=8.0134, tangentMode=RCTM_User, time=0, value=0.7}
         {arriveTangent=0.3471, interpMode=RCIM_Cubic, leaveTangent=0.3471, tangentMode=RCTM_User, time=0.1, value=1.2}
         {arriveTangent=-0.7173, interpMode=RCIM_Cubic, leaveTangent=-0.7173, tangentMode=RCTM_User, time=1, value=2}
        ]}}}
     CurveIndex=$Particles.NormalizedChainPosition
     Scale Curve=&Lerp_Float(A=40, B=60, Alpha=$System.SmoothDiveAlpha)
    )) {isSetParameters=true}
   &DynamicMaterialParameters(
    Index 0 Param 1=1
    Index 0 Param 2=&Add_Float(
     A=&Multiply_Float(A=$Engine.System.Age, B=1.2)
     B=&Multiply_FloatByInt(Float=3.14, Integer=$Particles.PartitionParticles.Partition)
    )
    Index 0 Param 3=1
    Index 0 Param 4=&Add_Float(
     A=&Multiply_Float(A=$Engine.System.Age, B=1.2)
     B=&Multiply_FloatByInt(Float=1.6902, Integer=$Particles.PartitionParticles.Partition)
    )
    Index 1 Param 1=1
    Index 1 Param 2=&Add_Float(
     A=&Multiply_Float(A=$Engine.System.Age, B=1.5786)
     B=&Multiply_FloatByInt(Float=1.6902, Integer=$Particles.PartitionParticles.Partition)
    )
    Index 1 Param 3=1
    Index 3 Param 1=1
    Index 3 Param 2=1
    Index 3 Param 3=1
    Module.Index0Param0Write=true
    Module.Index0Param1Write=true
    Module.Index0Param2Write=true
    Module.Index0Param3Write=true
    Module.Index1Param0Write=true
    Module.Index1Param1Write=true
    Module.Index1Param2Write=true
    Module.Index1Param3Write=true
    Module.Index3Param0Write=true
    Module.Index3Param1Write=true
    Module.Index3Param2Write=true
    Module.Index3Param3Write=true
    Write Parameter Index 1=true
    Write Parameter Index 3=true
   )
   &ScaleColor(
    Scale Alpha=&FloatFromCurve(
     FloatCurve=DataInterface() {propertyValues {Curve {keys [
         {arriveTangent=0.1505, interpMode=RCIM_Cubic, leaveTangent=0.1505, tangentMode=RCTM_User, time=0, value=1}
         {arriveTangent=-1, interpMode=RCIM_Cubic, leaveTangent=-1, time=0.4, value=0.8081}
         <at 1 0 RCIM_Cubic>
        ]}}}
     CurveIndex=$Particles.PartitionParticles.Normalized Index In Partition
    )
    Scale RGB=&MakeVectorFromLinearColorRGB(Color=&ColorFromCurve(
      ColorCurve=DataInterface() {propertyValues {
        AlphaCurve {keys [<at 0 1 RCIM_Cubic>,<at 1 1 RCIM_Cubic>]}
        BlueCurve {keys [<at 0 0.3>,<at 0.2193 0.0986>,<at 0.4643 0>,<at 0.7348 0.0986>,<at 1 0.3>]}
        GreenCurve {keys [<at 0 0.0347>,<at 0.2193 0.03>,<at 0.4643 0.1019>,<at 0.7348 0.03>,<at 1 0.0347>]}
        RedCurve {keys [<at 0 0.03>,<at 0.2193 0.3>,<at 0.4643 0.3>,<at 0.7348 0.3>,<at 1 0.03>]}
       }}
      CurveIndex=&Frac_Float(Float=&Add_Float(
        A=&Add_Float(
         A=&Multiply_Float(A=$Engine.System.Age, B=0.5)
         B=&Multiply_Float(A=$Particles.PartitionParticles.Normalized Index In Partition, B=-0.2)
        )
        B=/Niagara/DynamicInputs/Helpers/Waveform(Global Phase=$Engine.System.Age, "Waveform [1]"=::NewEnumerator12, "[1] Phase Scale"=0.4, "[1] Frequency"=1, "[1] Amplitude Min/Max"=<0 1>)
       ))
     ))
    ScaleA=true
    ScaleRGB=true
   ) {enabled=false}
   &self:Ribbon3Location(
    Amplitude:Vector3f=&Multiply_VectorByFloat(
     Vector=<18 18 1>
     Float=&FloatFromCurve(
      FloatCurve=DataInterface() {propertyValues {Curve {keys [
          {interpMode=RCIM_Cubic, tangentMode=RCTM_User, time=0, value=0}
          {arriveTangent=-2.3116, interpMode=RCIM_Cubic, leaveTangent=-2.3116, tangentMode=RCTM_User, time=0.4, value=1}
          <at 1 0>
         ]}}}
      CurveIndex=$Particles.PartitionParticles.Normalized Index In Partition
      Scale Curve=&Lerp_Float(A=1, B=1, Alpha=$System.SmoothDiveAlpha)
     )
    )
    Partition:NiagaraFloat=&MakeFloatFromInt(Int=$Particles.PartitionParticles.Partition)
    Scale:Vector3f=&Multiply_VectorByFloat(Vector=<1 1 1>, Float=0.5)
    Time:NiagaraFloat=&Add_Float(
     A=&Multiply_Float(A=$Engine.System.Age, B=-1.5)
     B=$Particles.PartitionParticles.Normalized Index In Partition
    )
   ) {module=Ribbon3Location}
   /SetVariables(
    Particles.Color02:LinearColor=&MakeLinearColorFromVectorAndFloat(
     "Vector (RGB)"=&Multiply_VectorByFloat(
      Vector=&MakeVectorFromLinearColorRGB(Color=&ColorFromCurve(
        ColorCurve=DataInterface() {propertyValues {
          AlphaCurve {keys [<at 0 1 RCIM_Cubic>,<at 1 1 RCIM_Cubic>]}
          BlueCurve {keys [<at 0 1.3>,<at 0.2787 2>,<at 0.4745 1.5>,<at 0.7559 0.1673>,<at 0.8762 0>,<at 1 0>]}
          GreenCurve {keys [<at 0 0.6828>,<at 0.2787 0.2779>,<at 0.4745 0>,<at 0.7559 0>,<at 0.8762 0.5146>,<at 1 0.8811>]}
          RedCurve {keys [<at 0 0>,<at 0.2787 0>,<at 0.4745 0.7049>,<at 0.7559 2>,<at 0.8762 2>,<at 1 2>]}
         }}
        CurveIndex=&Waveform(
         Global Input Position=&Multiply_FloatByInt(Float=$Particles.NormalizedChainPosition, Integer=1)
         Global Phase=&Add_Float(
          A=&Multiply_Float(A=$Engine.System.Age, B=-1)
          B=&Multiply_FloatByInt(Float=34, Integer=$Particles.PartitionParticles.Partition)
         )
         "Waveform [1]"=::NewEnumerator12
         "[1] Frequency"=0.8094
         "[1] Amplitude Min/Max"=<0 1>
        )
       ))
      Float=&Lerp_Float(A=0.4, B=&FN_FrontEndFloatModifier(Gameplay=0.9, Front End=0.5), Alpha=$System.SmoothDiveAlpha)
     )
     "Float (Alpha)"=&FloatFromCurve(
      FloatCurve=DataInterface() {propertyValues {Curve {keys [
          <at 0 1 RCIM_Cubic>
          {arriveTangent=-2.1381, interpMode=RCIM_Cubic, leaveTangent=-2.1381, tangentMode=RCTM_User, time=0.4255, value=0.5}
          <at 0.949 0 RCIM_Cubic>
         ]}}}
      CurveIndex=$Particles.NormalizedChainPosition
      Scale Curve=&FloatFromCurve(
       FloatCurve=DataInterface() {propertyValues {Curve {keys [{interpMode=RCIM_Cubic, tangentMode=RCTM_User, time=0, value=0}, {interpMode=RCIM_Cubic, tangentMode=RCTM_User, time=1, value=1}]}}}
       CurveIndex=$Engine.System.Age
       Scale Curve=2.5
      )
     )
    )
    Particles.DynamicMaterialParameter8:Vector4f=&MakeVector4(X=1, Y=1, Z=&Lerp_Float(A=3, B=4, Alpha=$System.SmoothDiveAlpha), W=0)
    Particles.RibbonFacing=/Niagara/DynamicInputs/Vectors/CrossProduct(Vector A:Vector3f=$Transient.PhysicsPendulumAxis, Vector B:Vector3f=<0 0 1>, Normalize Incoming Vectors:NiagaraBool=true)
    Particles.RibbonWidth2:NiagaraFloat=&Multiply_Float(A=$Particles.RibbonWidth, B=1)
   ) {isSetParameters=true}
   /SetVariables(
    Particles.Color3:LinearColor=&MultiplyLinearColor(
     Linear Color=$Particles.Color02
     Scale Factor=&MakeVector4_FromLinearColor(LinearColor=&LerpLinearColors(StartColor=<rgba 1 1 1 0.2>, EndColor=<rgba 10 10 10 1>, LerpFactor=$System.SmoothDiveAlpha))
    )
    Particles.DynamicMaterialParameter9:Vector4f=&MakeVector4(
     X=&Multiply_Float(
      A=$Engine.System.Age
      B=&Lerp_Float(A=-2, B=2, Alpha=&MakeFloatFromInt(Int=$Particles.Initial.PartitionParticles.Partition))
     )
     Y=&Lerp_Float(A=-3, B=3, Alpha=&MakeFloatFromInt(Int=$Particles.Initial.PartitionParticles.Partition))
     Z=&Multiply_Float(A=1, B=&Lerp_Float(A=0.4, B=1, Alpha=$System.SmoothDiveAlpha))
     W=&Lerp_Float(
      A=&FloatFromCurve(
       FloatCurve=DataInterface() {propertyValues {Curve {keys [{arriveTangent=-2.2554, interpMode=RCIM_Cubic, leaveTangent=-2.2554, tangentMode=RCTM_User, time=0, value=1}, <at 1 0>]}}}
       CurveIndex=$Particles.PartitionParticles.Normalized Index In Partition
      )
      B=&FloatFromCurve(
       FloatCurve=DataInterface() {propertyValues {Curve {keys [
           {arriveTangent=0.0761, interpMode=RCIM_Cubic, leaveTangent=0.0761, tangentMode=RCTM_User, time=0, value=1}
           {arriveTangent=-0.0852, interpMode=RCIM_Cubic, leaveTangent=-0.0852, tangentMode=RCTM_User, time=0.15, value=0.2}
           <at 1 0>
          ]}}}
       CurveIndex=$Particles.PartitionParticles.Normalized Index In Partition
       Scale Curve=10
      )
      Alpha=$System.SmoothDiveAlpha
     )
    )
    Particles.RibbonWidth3:NiagaraFloat=&Multiply_Float(
     A=$Particles.RibbonWidth
     B=&FloatFromCurve(
      FloatCurve=DataInterface() {propertyValues {Curve {keys [
          {arriveTangent=0.357, interpMode=RCIM_Cubic, leaveTangent=0.357, tangentMode=RCTM_User, time=0, value=0.3}
          {arriveTangent=-0.0652, interpMode=RCIM_Cubic, leaveTangent=-0.0652, tangentMode=RCTM_User, time=0.2, value=1}
          {arriveTangent=-0.1821, interpMode=RCIM_Cubic, leaveTangent=-0.1821, tangentMode=RCTM_User, time=0.7, value=0}
         ]}}}
      CurveIndex=$Particles.NormalizedChainPosition
      Scale Curve=&Lerp_Float(A=0.21, B=0.25, Alpha=$System.SmoothDiveAlpha)
     )
    )
   ) {isSetParameters=true}
   /SetVariables(
    Particles.DynamicMaterialParameter14:Vector4f=&MakeVector4(
     X=&FloatFromCurve(
      FloatCurve=DataInterface() {propertyValues {Curve {keys [
          <at 0 0.05 RCIM_Cubic>
          {arriveTangent=-0.1056, interpMode=RCIM_Cubic, leaveTangent=-0.1056, tangentMode=RCTM_User, time=0.0615, value=1}
          {arriveTangent=-2.6928, tangentMode=RCTM_User, time=1, value=0}
         ]}}}
      CurveIndex=$Particles.PartitionParticles.Normalized Index In Partition
      Scale Curve=&Lerp_Float(A=0.7, B=0.83, Alpha=$System.SmoothDiveAlpha)
     )
     Y=0
     Z=0
     W=0
    )
    Particles.RibbonWidth4:NiagaraFloat=&Multiply_Float(
     A=$Particles.RibbonWidth3
     B=&FloatFromCurve(
      FloatCurve=DataInterface() {propertyValues {Curve {keys [
          <at 0 1 RCIM_Cubic>
          {arriveTangent=-0.104, interpMode=RCIM_Cubic, leaveTangent=-0.104, tangentMode=RCTM_User, time=0.05, value=1.7}
          <at 1 1.2 RCIM_Cubic>
         ]}}}
      CurveIndex=$Particles.PartitionParticles.Normalized Index In Partition
     )
    )
   ) {isSetParameters=true}
   &self:NoiseLocation(
    Amplitude:Vector3f=&Multiply_VectorByFloat(
     Vector=<5 5 1>
     Float=&FloatFromCurve(
      FloatCurve=DataInterface() {propertyValues {Curve {keys [
          {interpMode=RCIM_Cubic, tangentMode=RCTM_User, time=0, value=0}
          {arriveTangent=-2.6919, interpMode=RCIM_Cubic, leaveTangent=-2.6919, tangentMode=RCTM_User, time=0.1, value=1}
          {tangentMode=RCTM_User, time=1, value=0}
         ]}}}
      CurveIndex=$Particles.PartitionParticles.Normalized Index In Partition
     )
    )
    Offset:Vector3f=&Multiply_VectorByFloat(Vector=<0 0 -2.2467>, Float=$Engine.System.Age)
    Scale:Vector3f=&Multiply_VectorByFloat(Vector=<1 1 1>, Float=100)
   ) {enabled=false, module=NoiseLocation}
  ]
  rendererClasses [&NiagaraRibbonRendererProperties]
  renderers [
   {rendererClass=&NiagaraRibbonRendererProperties}
   {rendererClass=&NiagaraRibbonRendererProperties, rendererIndex=1}
   {rendererClass=&NiagaraRibbonRendererProperties, rendererIndex=2}
  ]
 }
 NE_SplineThicken {
  emitterUpdateScript [
   &EmitterState(Life Cycle Mode=::NewEnumerator1, Loop Behavior=::NewEnumerator1, Loop Duration Mode=::NewEnumerator1, UseLoopDelay=false) {module=EmitterLifeCycle}
   &SpawnRate(
    Spawn Probability=0.8
    SpawnRate=&Lerp_Float(A=8, B=12, Alpha=$System.SmoothDiveAlpha)
    Use Spawn Probability=true
   )
  ]
  particleSpawnScript [
   &InitializeParticle(
    Lifetime=&Lerp_Float(A=1.3, B=1, Alpha=$System.SmoothDiveAlpha)
    Sprite Size Mode=::NewEnumerator3
    UsePositionOffset=false
    Write Material Random=true
   )
   &PartitionParticles()
   &InitialMeshOrientation(
    Mesh Orientation Mode=::NewEnumerator2
    Orientation Coordinate Space=::World
    Orientation Vector=<0 0 1>
    Rotation=&RandomRangeVector(Minimum=<0 0 0>, Maximum=<0 0 1>)
    Use Rotation Vector=true
   )
   /SetVariables(Particles.Scale=&Multiply_VectorByFloat(
     Vector=&Lerp_Vector(
      A=<0.7 0.7 2.5>
      B=<0.7 -0.7 2.5>
      Alpha=&MakeFloatFromInt(Int=$Particles.PartitionParticles.Partition)
     )
     Float=&Multiply_Float(
      A=&RandomRangeFloat(Minimum=0.4, Maximum=1)
      B=&Lerp_Float(A=0.6, B=1, Alpha=$System.SmoothDiveAlpha)
     )
    )) {isSetParameters=true}
  ]
  particleUpdateScript [
   &ParticleState()
   &ScaleMeshSize(Scale Factor=&VectorFromCurve(VectorCurve=DataInterface() {propertyValues {
       XCurve {keys [{arriveTangent=3.2161, interpMode=RCIM_Cubic, leaveTangent=3.2161, tangentMode=RCTM_User, time=0, value=0}, <at 1 1>]}
       YCurve {keys [{arriveTangent=3.2161, interpMode=RCIM_Cubic, leaveTangent=3.2161, tangentMode=RCTM_User, time=0, value=0}, <at 1 1>]}
       ZCurve {keys [{arriveTangent=3.2161, interpMode=RCIM_Cubic, leaveTangent=3.2161, tangentMode=RCTM_User, time=0, value=0}, <at 1 1>]}
      }}))
   &SolveForcesAndVelocity(Clamp Velocity=false, Limit Acceleration=false)
   &self:FollowRibbon1Path(
    Index:NiagaraInt32=$Particles.PartitionParticles.Index In Partition
    Offset:NiagaraFloat=&FloatFromCurve(
     FloatCurve=DataInterface() {propertyValues {Curve {keys [
         {arriveTangent=2.4606, interpMode=RCIM_Cubic, leaveTangent=2.4606, tangentMode=RCTM_User, time=0, value=0.1}
         {tangentMode=RCTM_User, time=1, value=1}
        ]}}}
     Scale Curve=0.5
    )
    Particle Attribute Reader=DataInterface() {propertyValues {EmitterBinding {bindingMode=Other, emitterName=NE_LocalSpaceRibbon}}}
    ParticlePerRibbon:NiagaraInt32=$System.ParticlePerRibbon
    Partition:NiagaraInt32=$Particles.PartitionParticles.Partition
   ) {module=FollowRibbon1Path}
   &UpdateMeshOrientation(
    Facing Direction=&AddVectorToPosition(
     Position=Linked() {linkedVariable {name=Particles.RibbonDirectionVector, type=Vector3f}}
     Vector=&Multiply_VectorByFloat(Vector=&RandomVector(Vector Scale=0.1), Float=1)
    )
    Facing Interpolation=false
    Orientation Method=::NewEnumerator1
    Override Quat=false
    TransformFacing=false
    TransformUp=false
    Up Direction:Vector3f=<0 1 0>
   )
   &UpdateMeshOrientation(
    Rotation Rate=&FloatFromCurve(
     FloatCurve=DataInterface() {propertyValues {Curve {keys [
         {arriveTangent=-1.4672, interpMode=RCIM_Cubic, leaveTangent=-1.4672, tangentMode=RCTM_User, time=0, value=1}
         {arriveTangent=-0.5831, tangentMode=RCTM_User, time=1, value=0}
        ]}}}
     Scale Curve=&Multiply_Float(
      A=&Lerp_Float(A=-1, B=1, Alpha=&MakeFloatFromInt(Int=$Particles.Initial.PartitionParticles.Partition))
      B=&Lerp_Float(A=2, B=4, Alpha=$System.SmoothDiveAlpha)
     )
    )
    Rotation Vector=<0 0 1>
   ) {module=UpdateMeshOrientation001}
   /SetVariables(Particles.Color=&MultiplyLinearColor(
     Linear Color=$Particles.RibbonColor
     Scale Factor=&MakeVector4_FromLinearColor(LinearColor=&ColorFromCurve(ColorCurve=DataInterface() {propertyValues {
         AlphaCurve {keys [<at 0.0069 0>,<at 0.1431 1>,<at 1 0 RCIM_Cubic>]}
         BlueCurve {keys [{arriveTangent=-42.5036, interpMode=RCIM_Cubic, leaveTangent=-42.5036, tangentMode=RCTM_User, time=0, value=20}, <at 1 0 RCIM_Cubic>]}
         GreenCurve {keys [{arriveTangent=-42.5036, interpMode=RCIM_Cubic, leaveTangent=-42.5036, tangentMode=RCTM_User, time=0, value=20}, <at 1 0 RCIM_Cubic>]}
         RedCurve {keys [{arriveTangent=-42.5036, interpMode=RCIM_Cubic, leaveTangent=-42.5036, tangentMode=RCTM_User, time=0, value=20}, <at 1 0 RCIM_Cubic>]}
        }}))
    )) {isSetParameters=true, module=SetVariables_#7}
   &DynamicMaterialParameters(
    Index 0 Param 1=&RandomRangeFloat(Minimum=5, Maximum=20)
    Index 0 Param 2=1
    Index 0 Param 3=&Multiply_Float(A=$Engine.System.Age, B=1)
    Module.Index0Param0Write=true
    Module.Index0Param1Write=true
    Module.Index0Param2Write=true
    Module.Index0Param3Write=true
   )
  ]
  rendererClasses [&NiagaraMeshRendererProperties]
  renderers [{rendererClass=&NiagaraMeshRendererProperties, rendererIndex=0}]
 }
 NE_SplineThicken001 {
  emitterUpdateScript [
   &EmitterState(Life Cycle Mode=::NewEnumerator1, Loop Behavior=::NewEnumerator1, Loop Duration Mode=::NewEnumerator1, UseLoopDelay=false) {module=EmitterLifeCycle}
   &SpawnRate(
    Spawn Probability=0.8
    SpawnRate=&Lerp_Float(A=8, B=30, Alpha=$System.SmoothDiveAlpha)
    Use Spawn Probability=true
   )
  ]
  particleSpawnScript [
   &InitializeParticle(
    Color Mode=::NewEnumerator0
    Lifetime=&Lerp_Float(A=0.5, B=0.5, Alpha=$System.SmoothDiveAlpha)
    Sprite Size Mode=::NewEnumerator3
    Uniform Sprite Size=&FloatFromCurve(
     FloatCurve=DataInterface() {propertyValues {Curve {keys [<at 0 15 RCIM_Cubic>, {arriveTangent=31.0503, tangentMode=RCTM_User, time=1, value=30}]}}}
     CurveIndex=&RandomRangeFloat(Minimum=0, Maximum=1)
    )
    UsePositionOffset=false
    Write Material Random=true
   )
   &PartitionParticles()
   &InitialMeshOrientation(
    Mesh Orientation Mode=::NewEnumerator2
    Orientation Coordinate Space=::World
    Orientation Vector=<0 0 1>
    Rotation=&RandomRangeVector(Minimum=<0 0 0>, Maximum=<0 0 1>)
    Use Rotation Vector=true
   )
  ]
  particleUpdateScript [
   &ParticleState()
   &ScaleSpriteSize(Uniform Curve Sprite Scale=DataInterface() {propertyValues {Curve {keys [
        {arriveTangent=6.8926, interpMode=RCIM_Cubic, leaveTangent=6.8926, tangentMode=RCTM_User, time=0, value=0}
        {interpMode=RCIM_Cubic, tangentMode=RCTM_User, time=0.2, value=1}
        {arriveTangent=-3.4257, tangentMode=RCTM_User, time=1, value=0}
       ]}}})
   &SolveForcesAndVelocity(Clamp Velocity=false, Limit Acceleration=false)
   &self:FollowRibbon1Path(
    Index:NiagaraInt32=$Particles.PartitionParticles.Index In Partition
    Offset:NiagaraFloat=&FloatFromCurve(
     FloatCurve=DataInterface() {propertyValues {Curve {keys [
         {arriveTangent=1.739, interpMode=RCIM_Cubic, leaveTangent=1.739, tangentMode=RCTM_User, time=0, value=0.3}
         {tangentMode=RCTM_User, time=1, value=1}
        ]}}}
     Scale Curve=&RandomRangeFloat(Minimum=0.1, Maximum=0.7)
    )
    Particle Attribute Reader=DataInterface() {propertyValues {EmitterBinding {bindingMode=Other, emitterName=NE_LocalSpaceRibbon}}}
    ParticlePerRibbon:NiagaraInt32=$System.ParticlePerRibbon
    Partition:NiagaraInt32=$Particles.PartitionParticles.Partition
   ) {module=FollowRibbon1Path}
   &self:NoiseLocation(
    Amplitude:Vector3f=&Multiply_VectorByFloat(
     Vector=<1 1 1>
     Float=&FloatFromCurve(
      FloatCurve=DataInterface() {propertyValues {Curve {keys [
          {interpMode=RCIM_Cubic, tangentMode=RCTM_User, time=0, value=0}
          {interpMode=RCIM_Cubic, tangentMode=RCTM_User, time=0.2, value=1}
          {tangentMode=RCTM_User, time=1, value=0}
         ]}}}
      Scale Curve=&RandomRangeFloat(Minimum=1, Maximum=4)
     )
    )
    Offset:Vector3f=&VectorFromFloat(Value=&RandomRangeFloat(Minimum=0, Maximum=200))
    Scale:Vector3f=&Multiply_VectorByFloat(Vector=<1 1 1>, Float=&RandomRangeFloat(Minimum=400, Maximum=400))
   ) {module=NoiseLocation}
   &UpdateMeshOrientation(
    Facing Direction=&AddVectorToPosition(
     Position=Linked() {linkedVariable {name=Particles.RibbonDirectionVector, type=Vector3f}}
     Vector=&Multiply_VectorByFloat(Vector=&RandomVector(Vector Scale=0.1), Float=1)
    )
    Facing Interpolation=false
    Orientation Method=::NewEnumerator1
    Override Quat=false
    TransformFacing=false
    TransformUp=false
    Up Direction:Vector3f=<0 1 0>
   )
   &UpdateMeshOrientation(
    Rotation Rate=&FloatFromCurve(
     FloatCurve=DataInterface() {propertyValues {Curve {keys [
         {arriveTangent=-1.4672, interpMode=RCIM_Cubic, leaveTangent=-1.4672, tangentMode=RCTM_User, time=0, value=1}
         {arriveTangent=-0.5831, tangentMode=RCTM_User, time=1, value=0}
        ]}}}
     Scale Curve=&Multiply_Float(
      A=&Lerp_Float(A=-1, B=1, Alpha=&MakeFloatFromInt(Int=$Particles.Initial.PartitionParticles.Partition))
      B=&Lerp_Float(A=2, B=4, Alpha=$System.SmoothDiveAlpha)
     )
    )
    Rotation Vector=<0 0 1>
   ) {module=UpdateMeshOrientation001}
   /SetVariables(Particles.Color=&MultiplyLinearColor(
     Linear Color=$Particles.RibbonColor
     Scale Factor=&MakeVector4_FromLinearColor(LinearColor=&ColorFromCurve(ColorCurve=DataInterface() {propertyValues {
         AlphaCurve {keys [<at 0.0069 0>,<at 0.1431 1>,<at 1 0 RCIM_Cubic>]}
         BlueCurve {keys [<at 0.0035 10>,<at 1 0 RCIM_Cubic>]}
         GreenCurve {keys [<at 0.0035 10>,<at 1 0 RCIM_Cubic>]}
         RedCurve {keys [<at 0.0035 10>,<at 1 0 RCIM_Cubic>]}
        }}))
    )) {isSetParameters=true, module=SetVariables_#7}
   &DynamicMaterialParameters(
    Index 0 Param 1=&RandomRangeFloat(Minimum=5, Maximum=20)
    Index 0 Param 2=1
    Index 0 Param 3=&Multiply_Float(A=$Engine.System.Age, B=1)
    Module.Index0Param0Write=true
    Module.Index0Param1Write=true
    Module.Index0Param2Write=true
    Module.Index0Param3Write=true
   )
   &CameraOffset(Camera Offset Amount=20)
  ]
  rendererClasses [&NiagaraSpriteRendererProperties]
  renderers [{rendererClass=&NiagaraSpriteRendererProperties}]
 }
 SkeletalMeshLocation_Skeleton {
  emitterUpdateScript [
   &EmitterState(Life Cycle Mode=::NewEnumerator1, Loop Behavior=::NewEnumerator1, Loop Duration Mode=::NewEnumerator1, UseLoopDelay=false)
   &SpawnBurst_Instantaneous(Spawn Count=2, Use Spawn Probability=false, UseLoopCountLimit=false)
  ]
  particleSpawnScript [
   &InitializeParticle(Color Mode=::NewEnumerator0, Lifetime=5, Sprite Size Mode=::NewEnumerator3, UsePositionOffset=false, Write Material Random=true)
   &PartitionParticles(Particles Per Partition=1, Partition Mode:ENiagara_PartitionMode=::NewEnumerator1)
  ]
  particleUpdateScript [
   &ParticleState(Kill Particles When Lifetime Has Elapsed=false)
   &self:SampleLocalRibbon(
    Particle Attribute Reader=DataInterface() {propertyValues {EmitterBinding {bindingMode=Other, emitterName=NE_LocalSpaceRibbon}}}
    ParticlePerRibbon:NiagaraInt32=$System.ParticlePerRibbon
    Partition:NiagaraInt32=$Particles.Initial.PartitionParticles.Partition
   ) {module=SampleLocalRibbon}
   &SkeletalMeshLocation(
    Bone / Socket Index=$Particles.PartitionParticles.Partition
    Bone Sampling Mode=::NewEnumerator4
    MaskSpawnGroup=false
    Sampled Position Offset=&Lerp_Vector(A=<0 0 0>, B=<0 0 0>, Alpha=&MakeFloatFromInt(Int=$Particles.PartitionParticles.Partition))
    Skeletal Mesh=$System.Skeletal Mesh
   ) {module=SkeletalMeshLocation001}
   &ScaleColor(
    Scale RGB=&MakeVectorFromLinearColorRGB(Color=&MultiplyLinearColor(
      Linear Color=$Output.SampleLocalRibbon.RibbonColor
      Scale Factor=&MakeVector4_FromLinearColor(LinearColor=&LerpLinearColors(StartColor=<rgba 0.2489 0.0287 0.3 1>, EndColor=<rgba 0.7378 0.7378 0.7378 1>, LerpFactor=$System.SmoothDiveAlpha))
     ))
    ScaleA=true
    ScaleRGB=true
   )
   &DynamicMaterialParameters(
    Index 0 Param 2=1
    Module.Index0Param0Write=true
    Module.Index0Param1Write=true
    Module.Index0Param2Write=true
    Module.Index0Param3Write=true
   )
  ]
  rendererClasses [&NiagaraMeshRendererProperties]
  renderers [{rendererClass=&NiagaraMeshRendererProperties, rendererIndex=0}]
 }
 Smoke_Wisp001 {
  emitterUpdateScript [
   &EmitterState()
   &SpawnRate(SpawnRate=&Lerp_Float(A=12, B=18, Alpha=$System.SmoothDiveAlpha), Use Spawn Probability=false)
  ]
  particleSpawnScript [
   &InitializeParticle(
    Lifetime=&Lerp_Float(A=1.5, B=1, Alpha=$System.SmoothDiveAlpha)
    Sprite Rotation Mode=::NewEnumerator2
    Sprite Size Mode=::NewEnumerator3
    Uniform Sprite Size=&Multiply_Float(
     A=&RandomRangeFloat(Minimum=80, Maximum=100)
     B=&Lerp_Float(A=1, B=1.4, Alpha=$System.SmoothDiveAlpha)
    )
    UsePositionOffset=false
    Write Material Random=true
   )
   &PartitionParticles()
   /SetVariables(Particles.UVScale:Vector2f=&Lerp_Vector2(A=<-1 1>, B=<1 1>, Alpha=&MakeFloatFromInt(Int=$Particles.PartitionParticles.Partition))) {isSetParameters=true}
  ]
  particleUpdateScript [
   &ParticleState()
   &self:FollowRibbon1Path(
    Offset:NiagaraFloat=&FloatFromCurve(
     FloatCurve=DataInterface() {propertyValues {Curve {keys [{arriveTangent=1.9463, interpMode=RCIM_Cubic, time=0, value=0}, {tangentMode=RCTM_User, time=1, value=1}]}}}
     Scale Curve=0.6
    )
    Particle Attribute Reader=DataInterface() {propertyValues {EmitterBinding {bindingMode=Other, emitterName=NE_LocalSpaceRibbon}}}
    ParticlePerRibbon:NiagaraInt32=$System.ParticlePerRibbon
    Partition:NiagaraInt32=$Particles.PartitionParticles.Partition
   ) {module=FollowRibbon1Path}
   &self:FollowRibbon3Path(
    Offset:NiagaraFloat=&FloatFromCurve(
     FloatCurve=DataInterface() {propertyValues {Curve {keys [{arriveTangent=1.9463, interpMode=RCIM_Cubic, time=0, value=0}, {tangentMode=RCTM_User, time=1, value=1}]}}}
     Scale Curve=0.6
    )
    Particle Attribute Reader=DataInterface() {propertyValues {EmitterBinding {bindingMode=Other, emitterName=NE_LocalSpaceRibbon}}}
    ParticlePerRibbon:NiagaraInt32=$System.ParticlePerRibbon
    Partition:NiagaraInt32=$Particles.PartitionParticles.Partition
   ) {enabled=false, module=FollowRibbon3Path}
   &ScaleColor(
    Scale Alpha=&FloatFromCurve(
     FloatCurve=DataInterface() {propertyValues {Curve {keys [{interpMode=RCIM_Cubic, tangentMode=RCTM_User, time=0, value=0}, <at 0.5 1 RCIM_Cubic>, {tangentMode=RCTM_User, time=1, value=0}]}}}
     Scale Curve=&Lerp_Float(A=0.5, B=0.3, Alpha=$System.SmoothDiveAlpha)
    )
    Scale RGB=&Lerp_Vector(
     A=&MakeVectorFromLinearColorRGB(Color=<rgba 0.324 0 0.2986 1>)
     B=&Multiply_VectorByFloat(Vector=&MakeVectorFromLinearColorRGB(Color=$Particles.RibbonColor), Float=0.5)
     Alpha=&Lerp_Float(A=0.1, B=0.5, Alpha=$System.SmoothDiveAlpha)
    )
    ScaleA=true
    ScaleRGB=true
   )
   &SpriteRotationRate(Rotation Rate=&Multiply_Float(
     A=&FloatFromCurve(FloatCurve=DataInterface() {propertyValues {Curve {keys [<at 0 1 RCIM_Cubic>,<at 1 0>]}}})
     B=&Lerp_Float(A=-180, B=180, Alpha=&MakeFloatFromInt(Int=$Particles.PartitionParticles.Partition))
    ))
   &ScaleSpriteSize(Scale Factor=&Vector2DFromFloat(Value=&FloatFromCurve(FloatCurve=DataInterface() {propertyValues {Curve {keys [
          {arriveTangent=1.4971, interpMode=RCIM_Cubic, leaveTangent=1.4971, tangentMode=RCTM_User, time=0, value=0}
          {tangentMode=RCTM_User, time=1, value=1}
         ]}}})))
   &SubUVAnimation(
    "Sprite Renderer "=DataInterface() {propertyValues {SpriteRenderer=&self:Smoke_Wisp001_1.NiagaraSpriteRendererProperties_1}}
    UseEndFrame=false
    UseStartFrame=false
   )
   &CameraOffset(Camera Offset Amount=-5)
   &DynamicMaterialParameters(
    Index 2 Param 1=2
    Index 2 Param 2=3
    Index 2 Param 3=3
    Module.Index0Param0Write=true
    Module.Index0Param1Write=true
    Module.Index0Param2Write=true
    Module.Index0Param3Write=true
    Module.Index1Param0Write=true
    Module.Index1Param1Write=true
    Module.Index1Param2Write=true
    Module.Index1Param3Write=true
    Module.Index2Param0Write=true
    Module.Index2Param1Write=true
    Module.Index2Param2Write=true
    Module.Index2Param3Write=true
    Write Parameter Index 1=true
    Write Parameter Index 2=true
   )
  ]
  rendererClasses [&NiagaraSpriteRendererProperties]
  renderers [{rendererClass=&NiagaraSpriteRendererProperties}]
 }
}
summary {
 systemName=NS_Contrail_PsychedelicSwirl
 emitters [
  {emitterName=NE_LocalSpaceRibbon, rendererClasses [&NiagaraRibbonRendererProperties]}
  {emitterName=NE_Empty001, rendererClasses [&NiagaraMeshRendererProperties]}
  {
   emitterName=Default_LightAndGlowFlicker_Loop
   rendererClasses [&NiagaraSpriteRendererProperties, &NiagaraLightRendererProperties]
  }
  {emitterName=NE_SplineThicken, rendererClasses [&NiagaraMeshRendererProperties]}
  {emitterName=Smoke_Wisp001, rendererClasses [&NiagaraSpriteRendererProperties]}
  {emitterName=NE_Empty, rendererClasses [&NiagaraMeshRendererProperties]}
  {emitterName=SkeletalMeshLocation_Skeleton, rendererClasses [&NiagaraMeshRendererProperties]}
  {emitterName=NE_SplineThicken001, rendererClasses [&NiagaraSpriteRendererProperties]}
 ]
 userVariables [
  User.bIsFrontEnd {type=NiagaraBool}
  User.bIsFrontEndPreview {type=NiagaraBool, defaultValue=true}
  User.Debug {type=NiagaraBool}
  User.ParaGlideLeanAlpha {type=NiagaraFloat, defaultValue=1, description=""}
 ]
}
